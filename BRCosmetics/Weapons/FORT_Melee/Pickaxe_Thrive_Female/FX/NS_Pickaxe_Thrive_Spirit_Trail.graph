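# Niagara anim-trail effect for the Thrive Female Spirit pickaxe, drawing green glowing ribbons between two bone or socket indices on the user-supplied skeletal mesh.
# Only NewTrail_Masked and NewTrail_Trans are enabled; the pink, hard-edged and Empty emitters are disabled.
# Both live ribbons offset ribbon IDs, drive U from normalized age, and use curves for color, alpha fade, width and dynamic material parameters.
# The Trans layer adds drag and colors by emitter age.
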NiagaraSystem /BRCosmetics/Weapons/FORT_Melee/Pickaxe_Thrive_Female/FX/NS_Pickaxe_Thrive_Spirit_Trail
dependencies {
 usedDataInterfaces [&NiagaraDataInterfaceColorCurve, &NiagaraDataInterfaceCurve]
 usedDynamicInputs []
 usedModules [&SystemState]
}
emitters {
 Empty005 {
  bEnabled=false
  emitterUpdateScript [
   &EmitterState()
   &SpawnRate(SpawnRate=25, Use Spawn Probability=false)
   &FN_SpawnAnimTrail_v2()
  ]
  particleSpawnScript [
   &InitializeParticle(Color Mode=::NewEnumerator0, UsePositionOffset=false, Write Material Random=true)
   &InitializeRibbon(Lifetime=5, Ribbon Width=50, UsePositionOffset=false, Write Material Random=true) {enabled=false}
   /Game/Effects/Niagara/Modules/AnimTrail/FN_InitializeAnimTrail/FN_InitializeAnimTrailBoneIndecies(First Bone or Socket Index=1, Second Bone or Socket Index=5)
   &FN_InitializeAnimTrail(Color=<rgba 1 1 1 0.02>, Lifetime=&RandomRangeFloat(Minimum=0.5, Maximum=0.5)) {module=FN_InitializeAnimTrail001}
   &AddVelocity()
  ]
  particleUpdateScript [
   &SolveForcesAndVelocity(Clamp Velocity=false, Limit Acceleration=false)
   &ParticleState()
   &FN_UpdateAnimTrail_v2()
   &Color(
    Color=&ColorFromCurve(ColorCurve=DataInterface() {propertyValues {
       AlphaCurve {keys [<at 0 1 RCIM_Cubic>,<at 1 1 RCIM_Cubic>]}
       BlueCurve {keys [
         <at 0 5>
         {arriveTangent=-31.1493, arriveTangentWeight=1.3418, leaveTangent=-31.1493, leaveTangentWeight=5.2828, time=0.1292, value=0.5017}
         <at 0.6377 0.1366 RCIM_Cubic>
        ]}
       GreenCurve {keys [
         <at 0 10>
         {arriveTangent=-30.2658, arriveTangentWeight=1.3037, leaveTangent=-30.2658, leaveTangentWeight=5.1331, time=0.1292, value=1}
         <at 0.6377 0.7 RCIM_Cubic>
        ]}
       RedCurve {keys [
         <at 0 5>
         {arriveTangent=-31.1523, arriveTangentWeight=1.3419, leaveTangent=-31.1523, leaveTangentWeight=5.2833, time=0.1292, value=0.5}
         <at 0.6377 0.1347 RCIM_Cubic>
        ]}
      }})
    Scale Alpha=&FloatFromCurve(FloatCurve=DataInterface() {propertyValues {Curve {keys [
         <at 0.1769 0.9943 RCIM_Cubic>
         {arriveTangent=-1.9828, interpMode=RCIM_Cubic, leaveTangent=-1.9828, tangentMode=RCTM_User, time=1.0011, value=0.002}
        ]}}})
   )
   &DynamicMaterialParameters(
    Index 0 Param 1=1
    Index 0 Param 2=1
    Index 0 Param 3=1
    Index 0 Param 4=1
    Index 2 Param 3=&FloatFromCurve(FloatCurve=DataInterface() {propertyValues {Curve {keys [
         {arriveTangent=-0.62, interpMode=RCIM_Cubic, leaveTangent=-0.62, tangentMode=RCTM_User, time=0, value=1}
         {arriveTangent=-1.9147, interpMode=RCIM_Cubic, leaveTangent=-1.9147, tangentMode=RCTM_User, time=1, value=0}
        ]}}})
    Index 2 Param 4=&FloatFromCurve(
     FloatCurve=DataInterface() {propertyValues {Curve {keys [
         {arriveTangent=-1.6564, interpMode=RCIM_Cubic, leaveTangent=-1.6564, tangentMode=RCTM_User, time=0, value=1}
         {interpMode=RCIM_Cubic, tangentMode=RCTM_User, time=1, value=0}
        ]}}}
     Scale Curve=2
    )
    Module.Param0WriteEnabled=true
    Module.Param0WriteEnabled1=true
    Module.Param0WriteEnabled2=true
    Module.Param0WriteEnabled3=true
    Module.Param1WriteEnabled=true
    Module.Param1WriteEnabled1=true
    Module.Param1WriteEnabled2=true
    Module.Param1WriteEnabled3=true
    Module.Param2WriteEnabled=true
    Module.Param2WriteEnabled1=true
    Module.Param2WriteEnabled2=true
    Module.Param2WriteEnabled3=true
    Module.Param3WriteEnabled=true
    Module.Param3WriteEnabled1=true
    Module.Param3WriteEnabled2=true
    Module.Param3WriteEnabled3=true
    Write Parameter Index 1=true
    Write Parameter Index 2=true
    Write Parameter Index 3=true
   )
   &ScaleRibbonWidth(Ribbon Width Scale=&Multiply_Float(
     A=&FloatFromCurve(FloatCurve=DataInterface() {propertyValues {Curve {keys [
          {arriveTangent=-0.9805, interpMode=RCIM_Cubic, leaveTangent=-0.9805, tangentMode=RCTM_User, time=0, value=1}
          {interpMode=RCIM_Cubic, tangentMode=RCTM_User, time=1, value=0.5}
         ]}}})
     B=&FloatFromCurve(
      FloatCurve=DataInterface() {propertyValues {Curve {keys [{interpMode=RCIM_Cubic, tangentMode=RCTM_User, time=0, value=1}, {interpMode=RCIM_Cubic, tangentMode=RCTM_User, time=1, value=1}]}}}
      CurveIndex=$Emitter.Age
     )
    ))
   &CameraOffset() {enabled=false}
  ]
  rendererClasses [&NiagaraRibbonRendererProperties]
  renderers [{rendererClass=&NiagaraRibbonRendererProperties}]
 }
 Empty007 {
  bEnabled=false
  emitterUpdateScript [
   &EmitterState()
   &SpawnRate(SpawnRate=25, Use Spawn Probability=false)
   &FN_SpawnAnimTrail_v2(AnimTrailWidthMode=::FromFirst)
  ]
  particleSpawnScript [
   &InitializeParticle(Color Mode=::NewEnumerator0, Lifetime=0.2, UsePositionOffset=false, Write Material Random=true) {enabled=false}
   &InitializeRibbon(Lifetime=5, Ribbon Width=50, UsePositionOffset=false, Write Material Random=true) {enabled=false}
   &FN_InitializeAnimTrailBoneIndecies(First Bone or Socket Index=1, Second Bone or Socket Index=5)
   &FN_InitializeAnimTrail(AnimTrailWidth=0.5, Color=<rgba 50 50 50 0.02>, Lifetime=1) {module=FN_InitializeAnimTrail001}
   &AddVelocity(Velocity=<0 0 20>)
  ]
  particleUpdateScript [
   &SolveForcesAndVelocity(Clamp Velocity=false, Limit Acceleration=false)
   &ParticleState()
   &FN_UpdateAnimTrail_v2()
   &Color(
    Color=&ColorFromCurve(ColorCurve=DataInterface() {propertyValues {
       AlphaCurve {keys [<at 0 1 RCIM_Cubic>,<at 1 1 RCIM_Cubic>]}
       BlueCurve {keys [<at 0 0.4508 RCIM_Cubic>,<at 1 0.4508 RCIM_Cubic>]}
       GreenCurve {keys [<at 0 0.0931 RCIM_Cubic>,<at 1 0.0931 RCIM_Cubic>]}
       RedCurve {keys [<at 0 0.6105 RCIM_Cubic>,<at 1 0.6105 RCIM_Cubic>]}
      }})
    Scale Alpha=&FloatFromCurve(FloatCurve=DataInterface() {propertyValues {Curve {keys [
         <at 0.1769 0.9943 RCIM_Cubic>
         {arriveTangent=-1.9828, interpMode=RCIM_Cubic, leaveTangent=-1.9828, tangentMode=RCTM_User, time=1.0011, value=0.002}
        ]}}})
   ) {enabled=false}
   &Color(
    Color=&ColorFromCurve(
     ColorCurve=DataInterface() {propertyValues {
       AlphaCurve {keys [<at 0 1 RCIM_Cubic>,<at 1 1 RCIM_Cubic>]}
       BlueCurve {keys [<at 0 1 RCIM_Cubic>,<at 0.2784 0.4>,<at 1 0.0931 RCIM_Cubic>]}
       GreenCurve {keys [<at 0 1 RCIM_Cubic>,<at 0.2784 1>,<at 1 0.6105 RCIM_Cubic>]}
       RedCurve {keys [<at 0 1 RCIM_Cubic>,<at 0.2784 0.4>,<at 1 0.0931 RCIM_Cubic>]}
      }}
     CurveIndex=$Emitter.Age
    )
    Scale Alpha=&FloatFromCurve(FloatCurve=DataInterface() {propertyValues {Curve {keys [
         <at 0.1769 0.9943 RCIM_Cubic>
         {arriveTangent=-1.9828, interpMode=RCIM_Cubic, leaveTangent=-1.9828, tangentMode=RCTM_User, time=1.0011, value=0.002}
        ]}}})
   ) {module=Color001}
   &DynamicMaterialParameters(
    Index 0 Param 1=1
    Index 0 Param 2=1
    Index 0 Param 3=1
    Index 0 Param 4=1
    Index 1 Param 1=1
    Index 1 Param 2=1
    Index 2 Param 1=1
    Index 2 Param 2=1
    Index 2 Param 3=&FloatFromCurve(FloatCurve=DataInterface() {propertyValues {Curve {keys [
         {arriveTangent=-2.9047, interpMode=RCIM_Cubic, leaveTangent=-2.9047, tangentMode=RCTM_User, time=0, value=1}
         {interpMode=RCIM_Cubic, tangentMode=RCTM_User, time=1, value=0}
        ]}}})
    Index 2 Param 4=&FloatFromCurve(FloatCurve=DataInterface() {propertyValues {Curve {keys [{interpMode=RCIM_Cubic, tangentMode=RCTM_User, time=0, value=1}, {interpMode=RCIM_Cubic, tangentMode=RCTM_User, time=1, value=0}]}}})
    Module.Param0WriteEnabled=true
    Module.Param0WriteEnabled1=true
    Module.Param0WriteEnabled2=true
    Module.Param0WriteEnabled3=true
    Module.Param1WriteEnabled=true
    Module.Param1WriteEnabled1=true
    Module.Param1WriteEnabled2=true
    Module.Param1WriteEnabled3=true
    Module.Param2WriteEnabled=true
    Module.Param2WriteEnabled1=true
    Module.Param2WriteEnabled2=true
    Module.Param2WriteEnabled3=true
    Module.Param3WriteEnabled=true
    Module.Param3WriteEnabled1=true
    Module.Param3WriteEnabled2=true
    Module.Param3WriteEnabled3=true
    Write Parameter Index 1=true
    Write Parameter Index 2=true
    Write Parameter Index 3=true
   )
   &ScaleRibbonWidth(Ribbon Width Scale=&FloatFromCurve(FloatCurve=DataInterface() {propertyValues {Curve {keys [
         {interpMode=RCIM_Cubic, tangentMode=RCTM_User, time=0, value=0.5}
         {arriveTangent=0.6917, interpMode=RCIM_Cubic, leaveTangent=0.6917, tangentMode=RCTM_User, time=1, value=2}
        ]}}}))
  ]
  rendererClasses [&NiagaraRibbonRendererProperties]
  renderers [{rendererClass=&NiagaraRibbonRendererProperties}]
 }
 NewTrail_Masked {
  emitterSpawnScript [&FN_RibbonIDOffsetUtility()]
  emitterUpdateScript [
   &EmitterState()
   &FN_SpawnAnimTrail()
  ]
  particleSpawnScript [
   /Game/Effects/Niagara/Modules/AnimTrail/FN_InitializeAnimTrailBoneIndecies_v2/FN_InitializeAnimTrailBoneIndecies(First Bone or Socket Index=1, Second Bone or Socket Index=5)
   &FN_InitializeAnimTrail(AnimTrailWidth=0.5, Color=<rgba 1 1 1 0.02>, Lifetime=0.2)
   &AddVelocity(Velocity=<0 5 0>)
  ]
  particleUpdateScript [
   &SolveForcesAndVelocity(Clamp Velocity=false, Limit Acceleration=false)
   &FN_UpdateAnimTrail(Locked Particle RibbonWidth=1, SkeletalMesh=$User.SkeletalMesh) {module=FN_UpdateAnimTrail001}
   &Color(
    Color=&ColorFromCurve(ColorCurve=DataInterface() {propertyValues {
       AlphaCurve {keys [<at 0 1 RCIM_Cubic>,<at 1 1 RCIM_Cubic>]}
       BlueCurve {keys [
         <at 0 5>
         {arriveTangent=-31.1493, arriveTangentWeight=1.3418, leaveTangent=-31.1493, leaveTangentWeight=5.2828, time=0.1292, value=0.5017}
         <at 0.6377 0.1366 RCIM_Cubic>
        ]}
       GreenCurve {keys [
         <at 0 10>
         {arriveTangent=-30.2658, arriveTangentWeight=1.3037, leaveTangent=-30.2658, leaveTangentWeight=5.1331, time=0.1292, value=1}
         <at 0.6377 0.7 RCIM_Cubic>
        ]}
       RedCurve {keys [
         <at 0 5>
         {arriveTangent=-31.1523, arriveTangentWeight=1.3419, leaveTangent=-31.1523, leaveTangentWeight=5.2833, time=0.1292, value=0.5}
         <at 0.6377 0.1347 RCIM_Cubic>
        ]}
      }})
    Scale Alpha=&FloatFromCurve(FloatCurve=DataInterface() {propertyValues {Curve {keys [
         <at 0.1769 0.9943 RCIM_Cubic>
         {arriveTangent=-1.9828, interpMode=RCIM_Cubic, leaveTangent=-1.9828, tangentMode=RCTM_User, time=1.0011, value=0.002}
        ]}}})
   )
   /SetVariables(Particles.RibbonU0Override=$Particles.NormalizedAge) {isSetParameters=true, module=SetVariables_#1}
   &DynamicMaterialParameters(
    Index 0 Param 1=1
    Index 0 Param 2=1
    Index 0 Param 3=1
    Index 0 Param 4=1
    Index 2 Param 3=&FloatFromCurve(FloatCurve=DataInterface() {propertyValues {Curve {keys [
         {arriveTangent=-0.62, interpMode=RCIM_Cubic, leaveTangent=-0.62, tangentMode=RCTM_User, time=0, value=1}
         {arriveTangent=-1.9147, interpMode=RCIM_Cubic, leaveTangent=-1.9147, tangentMode=RCTM_User, time=1, value=0}
        ]}}})
    Index 2 Param 4=&FloatFromCurve(
     FloatCurve=DataInterface() {propertyValues {Curve {keys [
         {arriveTangent=-1.6564, interpMode=RCIM_Cubic, leaveTangent=-1.6564, tangentMode=RCTM_User, time=0, value=1}
         {interpMode=RCIM_Cubic, tangentMode=RCTM_User, time=1, value=0}
        ]}}}
     Scale Curve=2
    )
    Module.Param0WriteEnabled=true
    Module.Param0WriteEnabled1=true
    Module.Param0WriteEnabled2=true
    Module.Param0WriteEnabled3=true
    Module.Param1WriteEnabled=true
    Module.Param1WriteEnabled1=true
    Module.Param1WriteEnabled2=true
    Module.Param1WriteEnabled3=true
    Module.Param2WriteEnabled=true
    Module.Param2WriteEnabled1=true
    Module.Param2WriteEnabled2=true
    Module.Param2WriteEnabled3=true
    Module.Param3WriteEnabled=true
    Module.Param3WriteEnabled1=true
    Module.Param3WriteEnabled2=true
    Module.Param3WriteEnabled3=true
    Write Parameter Index 1=true
    Write Parameter Index 2=true
    Write Parameter Index 3=true
   )
   &ScaleRibbonWidth(Ribbon Width Scale=&Multiply_Float(
     A=&FloatFromCurve(FloatCurve=DataInterface() {propertyValues {Curve {keys [
          {arriveTangent=-0.9805, interpMode=RCIM_Cubic, leaveTangent=-0.9805, tangentMode=RCTM_User, time=0, value=0}
          <at 0.0956 2>
          <at 0.302 0.5219>
          {interpMode=RCIM_Cubic, tangentMode=RCTM_User, time=1, value=0.5}
         ]}}})
     B=&FloatFromCurve(
      FloatCurve=DataInterface() {propertyValues {Curve {keys [{interpMode=RCIM_Cubic, tangentMode=RCTM_User, time=0, value=1}, {interpMode=RCIM_Cubic, tangentMode=RCTM_User, time=1, value=3}]}}}
      CurveIndex=$Emitter.Age
     )
    ))
  ]
  rendererClasses [&NiagaraRibbonRendererProperties]
  renderers [{rendererClass=&NiagaraRibbonRendererProperties}]
 }
 NewTrail_Trans {
  emitterSpawnScript [&FN_RibbonIDOffsetUtility()]
  emitterUpdateScript [
   &EmitterState()
   &FN_SpawnAnimTrail()
  ]
  particleSpawnScript [
   &FN_InitializeAnimTrailBoneIndecies(First Bone or Socket Index=1, Second Bone or Socket Index=5)
   &FN_InitializeAnimTrail(Color=<rgba 50 50 50 0.02>)
   &AddVelocity(Velocity=<0 40 10>)
  ]
  particleUpdateScript [
   &Drag(Use Linear Drag=true, Use Rotational Drag=false)
   &SolveForcesAndVelocity(Clamp Velocity=false, Limit Acceleration=false)
   &FN_UpdateAnimTrail(Locked Particle RibbonWidth=1, SkeletalMesh=$User.SkeletalMesh) {module=FN_UpdateAnimTrail001}
   &Color(
    Color=&ColorFromCurve(
     ColorCurve=DataInterface() {propertyValues {
       AlphaCurve {keys [<at 0 1 RCIM_Cubic>,<at 1 1 RCIM_Cubic>]}
       BlueCurve {keys [<at 0 1 RCIM_Cubic>,<at 0.2784 0.4>,<at 1 0.0931 RCIM_Cubic>]}
       GreenCurve {keys [<at 0 1 RCIM_Cubic>,<at 0.2784 1>,<at 1 0.6105 RCIM_Cubic>]}
       RedCurve {keys [<at 0 1 RCIM_Cubic>,<at 0.2784 0.4>,<at 1 0.0931 RCIM_Cubic>]}
      }}
     CurveIndex=$Emitter.Age
    )
    Scale Alpha=&FloatFromCurve(FloatCurve=DataInterface() {propertyValues {Curve {keys [
         <at 0.1769 0.9943 RCIM_Cubic>
         {arriveTangent=-1.9828, interpMode=RCIM_Cubic, leaveTangent=-1.9828, tangentMode=RCTM_User, time=1.0011, value=0.002}
        ]}}})
   ) {module=Color001}
   /SetVariables(Particles.RibbonU0Override=$Particles.NormalizedAge) {isSetParameters=true, module=SetVariables_#1}
   &DynamicMaterialParameters(
    Index 0 Param 1=1
    Index 0 Param 2=1
    Index 0 Param 3=1
    Index 0 Param 4=1
    Index 1 Param 1=1
    Index 1 Param 2=1
    Index 2 Param 1=1
    Index 2 Param 2=1
    Index 2 Param 3=&FloatFromCurve(FloatCurve=DataInterface() {propertyValues {Curve {keys [
         {arriveTangent=-2.9047, interpMode=RCIM_Cubic, leaveTangent=-2.9047, tangentMode=RCTM_User, time=0, value=1}
         {interpMode=RCIM_Cubic, tangentMode=RCTM_User, time=1, value=0}
        ]}}})
    Index 2 Param 4=&FloatFromCurve(FloatCurve=DataInterface() {propertyValues {Curve {keys [{interpMode=RCIM_Cubic, tangentMode=RCTM_User, time=0, value=1}, {interpMode=RCIM_Cubic, tangentMode=RCTM_User, time=1, value=0}]}}})
    Module.Param0WriteEnabled=true
    Module.Param0WriteEnabled1=true
    Module.Param0WriteEnabled2=true
    Module.Param0WriteEnabled3=true
    Module.Param1WriteEnabled=true
    Module.Param1WriteEnabled1=true
    Module.Param1WriteEnabled2=true
    Module.Param1WriteEnabled3=true
    Module.Param2WriteEnabled=true
    Module.Param2WriteEnabled1=true
    Module.Param2WriteEnabled2=true
    Module.Param2WriteEnabled3=true
    Module.Param3WriteEnabled=true
    Module.Param3WriteEnabled1=true
    Module.Param3WriteEnabled2=true
    Module.Param3WriteEnabled3=true
    Write Parameter Index 1=true
    Write Parameter Index 2=true
    Write Parameter Index 3=true
   )
   &ScaleRibbonWidth(Ribbon Width Scale=&Multiply_Float(
     A=&FloatFromCurve(
      FloatCurve=DataInterface() {propertyValues {Curve {keys [
          {interpMode=RCIM_Cubic, tangentMode=RCTM_User, time=0, value=0}
          {arriveTangent=-17, interpMode=RCIM_Cubic, leaveTangent=-17, time=0.05, value=4}
          {arriveTangent=-1.5, interpMode=RCIM_Cubic, leaveTangent=-1.5, time=0.2, value=0.6}
          {arriveTangent=0.6917, interpMode=RCIM_Cubic, leaveTangent=0.6917, tangentMode=RCTM_User, time=1, value=2}
         ]}}}
      Scale Curve=0.6
     )
     B=1
    ))
  ]
  rendererClasses [&NiagaraRibbonRendererProperties]
  renderers [{rendererClass=&NiagaraRibbonRendererProperties}]
 }
 PinkMasked_Original {
  bEnabled=false
  emitterUpdateScript [
   &EmitterState()
   &SpawnRate(SpawnRate=25, Use Spawn Probability=false)
   &FN_SpawnAnimTrail_v2(Spawn Mode:ENiagaraSkeletalSpawning=::NewEnumerator0)
  ]
  particleSpawnScript [
   &InitializeParticle(Color Mode=::NewEnumerator0, UsePositionOffset=false, Write Material Random=true)
   &InitializeRibbon(Lifetime=5, Ribbon Width=50, UsePositionOffset=false, Write Material Random=true) {enabled=false}
   /Game/Effects/Niagara/Modules/AnimTrail/FN_InitializeAnimTrail/FN_InitializeAnimTrailBoneIndecies(First Bone or Socket Index=1, Second Bone or Socket Index=5)
   &FN_InitializeAnimTrail(Color=<rgba 1 1 1 0.02>, Lifetime=&RandomRangeFloat(Minimum=0.5, Maximum=0.5)) {module=FN_InitializeAnimTrail001}
   &AddVelocity()
  ]
  particleUpdateScript [
   &SolveForcesAndVelocity(Clamp Velocity=false, Limit Acceleration=false)
   &ParticleState()
   &FN_UpdateAnimTrail_v2()
   &Color(
    Color=&ColorFromCurve(
     ColorCurve=DataInterface() {propertyValues {
       AlphaCurve {keys [<at 0 1 RCIM_Cubic>,<at 1 1 RCIM_Cubic>]}
       BlueCurve {keys [<at 0 0.8457 RCIM_Cubic>,<at 0.6377 0.4508 RCIM_Cubic>]}
       GreenCurve {keys [<at 0 0.5 RCIM_Cubic>,<at 0.6377 0.0931 RCIM_Cubic>]}
       RedCurve {keys [<at 0 1 RCIM_Cubic>,<at 0.6377 0.6105 RCIM_Cubic>]}
      }}
     CurveIndex=$Emitter.Age
    )
    Scale Alpha=&FloatFromCurve(FloatCurve=DataInterface() {propertyValues {Curve {keys [
         <at 0.1769 0.9943 RCIM_Cubic>
         {arriveTangent=-1.9828, interpMode=RCIM_Cubic, leaveTangent=-1.9828, tangentMode=RCTM_User, time=1.0011, value=0.002}
        ]}}})
   )
   &DynamicMaterialParameters(
    Index 0 Param 1=1
    Index 0 Param 2=1
    Index 0 Param 3=1
    Index 0 Param 4=1
    Index 2 Param 3=&FloatFromCurve(FloatCurve=DataInterface() {propertyValues {Curve {keys [
         {arriveTangent=-0.62, interpMode=RCIM_Cubic, leaveTangent=-0.62, tangentMode=RCTM_User, time=0, value=1}
         {arriveTangent=-1.9147, interpMode=RCIM_Cubic, leaveTangent=-1.9147, tangentMode=RCTM_User, time=1, value=0}
        ]}}})
    Index 2 Param 4=&FloatFromCurve(
     FloatCurve=DataInterface() {propertyValues {Curve {keys [
         {arriveTangent=-1.6564, interpMode=RCIM_Cubic, leaveTangent=-1.6564, tangentMode=RCTM_User, time=0, value=1}
         {interpMode=RCIM_Cubic, tangentMode=RCTM_User, time=1, value=0}
        ]}}}
     Scale Curve=2
    )
    Module.Param0WriteEnabled=true
    Module.Param0WriteEnabled1=true
    Module.Param0WriteEnabled2=true
    Module.Param0WriteEnabled3=true
    Module.Param1WriteEnabled=true
    Module.Param1WriteEnabled1=true
    Module.Param1WriteEnabled2=true
    Module.Param1WriteEnabled3=true
    Module.Param2WriteEnabled=true
    Module.Param2WriteEnabled1=true
    Module.Param2WriteEnabled2=true
    Module.Param2WriteEnabled3=true
    Module.Param3WriteEnabled=true
    Module.Param3WriteEnabled1=true
    Module.Param3WriteEnabled2=true
    Module.Param3WriteEnabled3=true
    Write Parameter Index 1=true
    Write Parameter Index 2=true
    Write Parameter Index 3=true
   )
   &ScaleRibbonWidth(Ribbon Width Scale=&Multiply_Float(
     A=&FloatFromCurve(FloatCurve=DataInterface() {propertyValues {Curve {keys [
          {arriveTangent=-0.9805, interpMode=RCIM_Cubic, leaveTangent=-0.9805, tangentMode=RCTM_User, time=0, value=1}
          {interpMode=RCIM_Cubic, tangentMode=RCTM_User, time=1, value=0.5}
         ]}}})
     B=&FloatFromCurve(
      FloatCurve=DataInterface() {propertyValues {Curve {keys [{interpMode=RCIM_Cubic, tangentMode=RCTM_User, time=0, value=1}, {interpMode=RCIM_Cubic, tangentMode=RCTM_User, time=1, value=1}]}}}
      CurveIndex=$Emitter.Age
     )
    ))
   &CameraOffset() {enabled=false}
  ]
  rendererClasses [&NiagaraRibbonRendererProperties]
  renderers [{rendererClass=&NiagaraRibbonRendererProperties}]
 }
 PinkTrans_Original {
  bEnabled=false
  emitterUpdateScript [
   &EmitterState()
   &SpawnRate(SpawnRate=25, Use Spawn Probability=false)
   &FN_SpawnAnimTrail_v2(Spawn Mode:ENiagaraSkeletalSpawning=::NewEnumerator0)
  ]
  particleSpawnScript [
   &InitializeParticle(Color Mode=::NewEnumerator0, Lifetime=0.2, UsePositionOffset=false, Write Material Random=true) {enabled=false}
   &InitializeRibbon(Lifetime=5, Ribbon Width=50, UsePositionOffset=false, Write Material Random=true) {enabled=false}
   &FN_InitializeAnimTrailBoneIndecies(First Bone or Socket Index=1, Second Bone or Socket Index=3)
   &FN_InitializeAnimTrail(AnimTrailWidth=0.5, Color=<rgba 50 50 50 0.02>, Lifetime=1) {module=FN_InitializeAnimTrail001}
   &AddVelocity(Velocity=<0 0 20>)
  ]
  particleUpdateScript [
   &SolveForcesAndVelocity(Clamp Velocity=false, Limit Acceleration=false)
   &ParticleState()
   &FN_UpdateAnimTrail_v2()
   &Color(
    Color=&ColorFromCurve(ColorCurve=DataInterface() {propertyValues {
       AlphaCurve {keys [<at 0 1 RCIM_Cubic>,<at 1 1 RCIM_Cubic>]}
       BlueCurve {keys [<at 0 0.4508 RCIM_Cubic>,<at 1 0.4508 RCIM_Cubic>]}
       GreenCurve {keys [<at 0 0.0931 RCIM_Cubic>,<at 1 0.0931 RCIM_Cubic>]}
       RedCurve {keys [<at 0 0.6105 RCIM_Cubic>,<at 1 0.6105 RCIM_Cubic>]}
      }})
    Scale Alpha=&FloatFromCurve(FloatCurve=DataInterface() {propertyValues {Curve {keys [
         <at 0.1769 0.9943 RCIM_Cubic>
         {arriveTangent=-1.9828, interpMode=RCIM_Cubic, leaveTangent=-1.9828, tangentMode=RCTM_User, time=1.0011, value=0.002}
        ]}}})
   ) {enabled=false}
   &Color(
    Color=&ColorFromCurve(
     ColorCurve=DataInterface() {propertyValues {
       AlphaCurve {keys [<at 0 1 RCIM_Cubic>,<at 1 1 RCIM_Cubic>]}
       BlueCurve {keys [<at 0 0.7897 RCIM_Cubic>,<at 0.2784 0.8148>,<at 1 0.4508 RCIM_Cubic>]}
       GreenCurve {keys [<at 0 0.6 RCIM_Cubic>,<at 0.2784 0.4>,<at 1 0.0931 RCIM_Cubic>]}
       RedCurve {keys [<at 0 1 RCIM_Cubic>,<at 0.2784 1>,<at 1 0.6105 RCIM_Cubic>]}
      }}
     CurveIndex=$Emitter.Age
    )
    Scale Alpha=&FloatFromCurve(FloatCurve=DataInterface() {propertyValues {Curve {keys [
         <at 0.1769 0.9943 RCIM_Cubic>
         {arriveTangent=-1.9828, interpMode=RCIM_Cubic, leaveTangent=-1.9828, tangentMode=RCTM_User, time=1.0011, value=0.002}
        ]}}})
   ) {module=Color001}
   &DynamicMaterialParameters(
    Index 0 Param 1=1
    Index 0 Param 2=1
    Index 0 Param 3=1
    Index 0 Param 4=1
    Index 1 Param 1=1
    Index 1 Param 2=1
    Index 2 Param 1=1
    Index 2 Param 2=1
    Index 2 Param 3=&FloatFromCurve(FloatCurve=DataInterface() {propertyValues {Curve {keys [
         {arriveTangent=-2.9047, interpMode=RCIM_Cubic, leaveTangent=-2.9047, tangentMode=RCTM_User, time=0, value=1}
         {interpMode=RCIM_Cubic, tangentMode=RCTM_User, time=1, value=0}
        ]}}})
    Index 2 Param 4=&FloatFromCurve(FloatCurve=DataInterface() {propertyValues {Curve {keys [{interpMode=RCIM_Cubic, tangentMode=RCTM_User, time=0, value=1}, {interpMode=RCIM_Cubic, tangentMode=RCTM_User, time=1, value=0}]}}})
    Module.Param0WriteEnabled=true
    Module.Param0WriteEnabled1=true
    Module.Param0WriteEnabled2=true
    Module.Param0WriteEnabled3=true
    Module.Param1WriteEnabled=true
    Module.Param1WriteEnabled1=true
    Module.Param1WriteEnabled2=true
    Module.Param1WriteEnabled3=true
    Module.Param2WriteEnabled=true
    Module.Param2WriteEnabled1=true
    Module.Param2WriteEnabled2=true
    Module.Param2WriteEnabled3=true
    Module.Param3WriteEnabled=true
    Module.Param3WriteEnabled1=true
    Module.Param3WriteEnabled2=true
    Module.Param3WriteEnabled3=true
    Write Parameter Index 1=true
    Write Parameter Index 2=true
    Write Parameter Index 3=true
   )
   &ScaleRibbonWidth(Ribbon Width Scale=&FloatFromCurve(FloatCurve=DataInterface() {propertyValues {Curve {keys [
         {interpMode=RCIM_Cubic, tangentMode=RCTM_User, time=0, value=0.5}
         {arriveTangent=0.6917, interpMode=RCIM_Cubic, leaveTangent=0.6917, tangentMode=RCTM_User, time=1, value=2}
        ]}}})) {enabled=false}
  ]
  rendererClasses [&NiagaraRibbonRendererProperties]
  renderers [{rendererClass=&NiagaraRibbonRendererProperties}]
 }
 PinkTrans_Original001 {
  bEnabled=false
  emitterUpdateScript [
   &EmitterState()
   &SpawnRate(SpawnRate=25, Use Spawn Probability=false)
   &FN_SpawnAnimTrail_v2()
  ]
  particleSpawnScript [
   &InitializeParticle(Color Mode=::NewEnumerator0, Lifetime=0.2, UsePositionOffset=false, Write Material Random=true) {enabled=false}
   &InitializeRibbon(Lifetime=5, Ribbon Width=50, UsePositionOffset=false, Write Material Random=true) {enabled=false}
   &FN_InitializeAnimTrailBoneIndecies(First Bone or Socket Index=1, Second Bone or Socket Index=5)
   &FN_InitializeAnimTrail(AnimTrailWidth=0.5, Color=<rgba 50 50 50 0.02>, Lifetime=1) {module=FN_InitializeAnimTrail001}
   &AddVelocity(Velocity=<0 0 20>)
  ]
  particleUpdateScript [
   &SolveForcesAndVelocity(Clamp Velocity=false, Limit Acceleration=false)
   &ParticleState()
   &FN_UpdateAnimTrail_v2()
   &Color(
    Color=&ColorFromCurve(ColorCurve=DataInterface() {propertyValues {
       AlphaCurve {keys [<at 0 1 RCIM_Cubic>,<at 1 1 RCIM_Cubic>]}
       BlueCurve {keys [<at 0 0.4508 RCIM_Cubic>,<at 1 0.4508 RCIM_Cubic>]}
       GreenCurve {keys [<at 0 0.0931 RCIM_Cubic>,<at 1 0.0931 RCIM_Cubic>]}
       RedCurve {keys [<at 0 0.6105 RCIM_Cubic>,<at 1 0.6105 RCIM_Cubic>]}
      }})
    Scale Alpha=&FloatFromCurve(FloatCurve=DataInterface() {propertyValues {Curve {keys [
         <at 0.1769 0.9943 RCIM_Cubic>
         {arriveTangent=-1.9828, interpMode=RCIM_Cubic, leaveTangent=-1.9828, tangentMode=RCTM_User, time=1.0011, value=0.002}
        ]}}})
   ) {enabled=false}
   &Color(
    Color=&ColorFromCurve(
     ColorCurve=DataInterface() {propertyValues {
       AlphaCurve {keys [<at 0 1 RCIM_Cubic>,<at 1 1 RCIM_Cubic>]}
       BlueCurve {keys [<at 0 0.7897 RCIM_Cubic>,<at 0.2784 0.8148>,<at 1 0.4508 RCIM_Cubic>]}
       GreenCurve {keys [<at 0 0.6 RCIM_Cubic>,<at 0.2784 0.4>,<at 1 0.0931 RCIM_Cubic>]}
       RedCurve {keys [<at 0 1 RCIM_Cubic>,<at 0.2784 1>,<at 1 0.6105 RCIM_Cubic>]}
      }}
     CurveIndex=$Emitter.Age
    )
    Scale Alpha=&FloatFromCurve(FloatCurve=DataInterface() {propertyValues {Curve {keys [
         <at 0.1769 0.9943 RCIM_Cubic>
         {arriveTangent=-1.9828, interpMode=RCIM_Cubic, leaveTangent=-1.9828, tangentMode=RCTM_User, time=1.0011, value=0.002}
        ]}}})
   ) {module=Color001}
   &DynamicMaterialParameters(
    Index 0 Param 1=1
    Index 0 Param 2=1
    Index 0 Param 3=1
    Index 0 Param 4=1
    Index 1 Param 1=1
    Index 1 Param 2=1
    Index 2 Param 1=1
    Index 2 Param 2=1
    Index 2 Param 3=&FloatFromCurve(FloatCurve=DataInterface() {propertyValues {Curve {keys [
         {arriveTangent=-2.9047, interpMode=RCIM_Cubic, leaveTangent=-2.9047, tangentMode=RCTM_User, time=0, value=1}
         {interpMode=RCIM_Cubic, tangentMode=RCTM_User, time=1, value=0}
        ]}}})
    Index 2 Param 4=&FloatFromCurve(FloatCurve=DataInterface() {propertyValues {Curve {keys [{interpMode=RCIM_Cubic, tangentMode=RCTM_User, time=0, value=1}, {interpMode=RCIM_Cubic, tangentMode=RCTM_User, time=1, value=0}]}}})
    Module.Param0WriteEnabled=true
    Module.Param0WriteEnabled1=true
    Module.Param0WriteEnabled2=true
    Module.Param0WriteEnabled3=true
    Module.Param1WriteEnabled=true
    Module.Param1WriteEnabled1=true
    Module.Param1WriteEnabled2=true
    Module.Param1WriteEnabled3=true
    Module.Param2WriteEnabled=true
    Module.Param2WriteEnabled1=true
    Module.Param2WriteEnabled2=true
    Module.Param2WriteEnabled3=true
    Module.Param3WriteEnabled=true
    Module.Param3WriteEnabled1=true
    Module.Param3WriteEnabled2=true
    Module.Param3WriteEnabled3=true
    Write Parameter Index 1=true
    Write Parameter Index 2=true
    Write Parameter Index 3=true
   )
   &ScaleRibbonWidth(Ribbon Width Scale=&FloatFromCurve(FloatCurve=DataInterface() {propertyValues {Curve {keys [
         {interpMode=RCIM_Cubic, tangentMode=RCTM_User, time=0, value=0.5}
         {arriveTangent=0.6917, interpMode=RCIM_Cubic, leaveTangent=0.6917, tangentMode=RCTM_User, time=1, value=2}
        ]}}})) {enabled=false}
  ]
  rendererClasses [&NiagaraRibbonRendererProperties]
  renderers [{rendererClass=&NiagaraRibbonRendererProperties}]
 }
 Trails_HardEdged {
  bEnabled=false
  emitterUpdateScript [
   &EmitterState()
   &SpawnRate(SpawnRate=25, Use Spawn Probability=false) {enabled=false}
   &FN_SpawnAnimTrail_v2()
  ]
  particleSpawnScript [
   &InitializeParticle(Color Mode=::NewEnumerator0, UsePositionOffset=false, Write Material Random=true)
   &InitializeRibbon(Lifetime=5, Ribbon Width=50, UsePositionOffset=false, Write Material Random=true) {enabled=false}
   &FN_InitializeAnimTrailBoneIndecies(First Bone or Socket Index=1, Second Bone or Socket Index=5)
   &FN_InitializeAnimTrail(AnimTrailWidth=0.5, Color=<rgba 1 1 1 0.02>, Lifetime=&RandomRangeFloat(Minimum=0.5, Maximum=0.7)) {module=FN_InitializeAnimTrail001}
   &AddVelocity()
  ]
  particleUpdateScript [
   &ParticleState()
   &FN_UpdateAnimTrail_v2()
   &Color(
    Color=&ColorFromCurve(
     ColorCurve=DataInterface() {propertyValues {
       AlphaCurve {keys [<at 0 1 RCIM_Cubic>,<at 1 1 RCIM_Cubic>]}
       BlueCurve {keys [<at 0 0.8457 RCIM_Cubic>,<at 0.6377 0.4508 RCIM_Cubic>]}
       GreenCurve {keys [<at 0 0.5 RCIM_Cubic>,<at 0.6377 0.0931 RCIM_Cubic>]}
       RedCurve {keys [<at 0 1 RCIM_Cubic>,<at 0.6377 0.6105 RCIM_Cubic>]}
      }}
     CurveIndex=$Emitter.Age
    )
    Scale Alpha=&FloatFromCurve(FloatCurve=DataInterface() {propertyValues {Curve {keys [
         <at 0.1769 0.9943 RCIM_Cubic>
         {arriveTangent=-1.9828, interpMode=RCIM_Cubic, leaveTangent=-1.9828, tangentMode=RCTM_User, time=1.0011, value=0.002}
        ]}}})
   )
   &self:FadeAlpha_head() {enabled=false, module=FadeAlpha_head}
   &DynamicMaterialParameters(
    Index 0 Param 1=1
    Index 0 Param 2=1
    Index 0 Param 3=1
    Index 0 Param 4=1
    Index 2 Param 3=&FloatFromCurve(FloatCurve=DataInterface() {propertyValues {Curve {keys [
         {arriveTangent=-0.62, interpMode=RCIM_Cubic, leaveTangent=-0.62, tangentMode=RCTM_User, time=0, value=1}
         {arriveTangent=-1.9147, interpMode=RCIM_Cubic, leaveTangent=-1.9147, tangentMode=RCTM_User, time=1, value=0}
        ]}}})
    Index 2 Param 4=&FloatFromCurve(
     FloatCurve=DataInterface() {propertyValues {Curve {keys [
         {arriveTangent=-1.6564, interpMode=RCIM_Cubic, leaveTangent=-1.6564, tangentMode=RCTM_User, time=0, value=1}
         {interpMode=RCIM_Cubic, tangentMode=RCTM_User, time=1, value=0}
        ]}}}
     Scale Curve=2
    )
    Module.Param0WriteEnabled=true
    Module.Param0WriteEnabled1=true
    Module.Param0WriteEnabled2=true
    Module.Param0WriteEnabled3=true
    Module.Param1WriteEnabled=true
    Module.Param1WriteEnabled1=true
    Module.Param1WriteEnabled2=true
    Module.Param1WriteEnabled3=true
    Module.Param2WriteEnabled=true
    Module.Param2WriteEnabled1=true
    Module.Param2WriteEnabled2=true
    Module.Param2WriteEnabled3=true
    Module.Param3WriteEnabled=true
    Module.Param3WriteEnabled1=true
    Module.Param3WriteEnabled2=true
    Module.Param3WriteEnabled3=true
    Write Parameter Index 1=true
    Write Parameter Index 2=true
    Write Parameter Index 3=true
   )
   &ScaleRibbonWidth(Ribbon Width Scale=&Multiply_Float(
     A=&FloatFromCurve(FloatCurve=DataInterface() {propertyValues {Curve {keys [
          {arriveTangent=-0.9805, interpMode=RCIM_Cubic, leaveTangent=-0.9805, tangentMode=RCTM_User, time=0, value=1}
          {interpMode=RCIM_Cubic, tangentMode=RCTM_User, time=1, value=0.5}
         ]}}})
     B=&FloatFromCurve(
      FloatCurve=DataInterface() {propertyValues {Curve {keys [{interpMode=RCIM_Cubic, tangentMode=RCTM_User, time=0, value=1}, {interpMode=RCIM_Cubic, tangentMode=RCTM_User, time=1, value=1}]}}}
      CurveIndex=$Emitter.Age
     )
    )) {enabled=false}
   &self:ScaleWidth_head(Power=6) {module=ScaleWidth_head}
   &SolveForcesAndVelocity(Clamp Velocity=false, Limit Acceleration=false)
   &CameraOffset() {enabled=false}
  ]
  rendererClasses [&NiagaraRibbonRendererProperties]
  renderers [{rendererClass=&NiagaraRibbonRendererProperties}]
 }
}
summary {
 systemName=NS_Pickaxe_Thrive_Spirit_Trail
 emitters [
  {bEnabled=false, emitterName=Empty005, rendererClasses [&NiagaraRibbonRendererProperties]}
  {bEnabled=false, emitterName=Empty007, rendererClasses [&NiagaraRibbonRendererProperties]}
  {bEnabled=false, emitterName=PinkMasked_Original, rendererClasses [&NiagaraRibbonRendererProperties]}
  {bEnabled=false, emitterName=PinkTrans_Original, rendererClasses [&NiagaraRibbonRendererProperties]}
  {bEnabled=false, emitterName=PinkTrans_Original001, rendererClasses [&NiagaraRibbonRendererProperties]}
  {bEnabled=false, emitterName=Trails_HardEdged, rendererClasses [&NiagaraRibbonRendererProperties]}
  {emitterName=NewTrail_Masked, rendererClasses [&NiagaraRibbonRendererProperties]}
  {emitterName=NewTrail_Trans, rendererClasses [&NiagaraRibbonRendererProperties]}
 ]
 userVariables [
  User.Decor Color {type=LinearColor, description="", defaultValue {struct=LinearColor, value=<rgba 0.4125 0.0976 0.0704 1>}}
  User.Glow Color {type=LinearColor, description="", defaultValue {struct=LinearColor, value=<rgba 0.1925 1 0.1952 1>}}
  User.Handle Color {type=LinearColor, description="", defaultValue {struct=LinearColor, value=<rgba 0.2623 0.1651 0.117 1>}}
  User.PhysicalSurface {type=NiagaraInt32}
  User.RibbonIDOffset {type=NiagaraInt32}
  User.SkeletalMesh {
   type=NiagaraDataInterfaceSkeletalMesh
   defaultValue {
    struct=NiagaraExt_VariableValue_DataInterface
    value {dataInterface=&self:NiagaraDataInterfaceSkeletalMesh_1}
   }
  }
 ]
}
