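# Physics asset for the Thrive Female pickaxe, a two-handed melee cosmetic, covering its handle, root and dangling-style crystal bones.
# Each bone has an identical small capsule body, chained from b_AR_Root through Melee_TwoHanded_Handle to dyn_center.
# Five dyn_crystal bodies branch from dyn_center with limited swing and twist.
# The root-to-handle and handle-to-center joints leave all angular motion free, so their stored limit angles do not apply.
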PhysicsAsset /BRCosmetics/Weapons/FORT_Melee/Pickaxe_Thrive_Female/Meshes/Thrive_Female_Axe_PhysicsAsset
bodies {
 Melee_TwoHanded_Handle [{shapeType=Capsule, length=2, radius=2, shapeName=Melee_TwoHanded_Handle_capsule}]
 b_AR_Root [{shapeType=Capsule, length=2, radius=2, shapeName=b_AR_Root_capsule}]
 dyn_center [{shapeType=Capsule, length=2, radius=2, shapeName=dyn_center_capsule}]
 dyn_crystal_01 [{shapeType=Capsule, length=2, radius=2, shapeName=dyn_crystal_01_capsule}]
 dyn_crystal_02 [{shapeType=Capsule, length=2, radius=2, shapeName=dyn_crystal_02_capsule}]
 dyn_crystal_03 [{shapeType=Capsule, length=2, radius=2, shapeName=dyn_crystal_03_capsule}]
 dyn_crystal_04 [{shapeType=Capsule, length=2, radius=2, shapeName=dyn_crystal_04_capsule}]
 dyn_crystal_05 [{shapeType=Capsule, length=2, radius=2, shapeName=dyn_crystal_05_capsule}]
}
constraints [
 {bone1Name=dyn_crystal_05, bone2Name=dyn_center, swing1LimitDegrees=30, swing2LimitDegrees=30, twistLimitDegrees=45}
 {
  bone1Name=Melee_TwoHanded_Handle
  bone2Name=b_AR_Root
  swing1LimitDegrees=45
  swing1Motion=Free
  swing2LimitDegrees=45
  swing2Motion=Free
  twistLimitDegrees=45
  twistMotion=Free
 }
 {
  bone1Name=dyn_center
  bone2Name=Melee_TwoHanded_Handle
  swing1LimitDegrees=45
  swing1Motion=Free
  swing2LimitDegrees=45
  swing2Motion=Free
  twistLimitDegrees=45
  twistMotion=Free
 }
 {bone1Name=dyn_crystal_01, bone2Name=dyn_center, swing1LimitDegrees=30, swing2LimitDegrees=30, twistLimitDegrees=45}
 {bone1Name=dyn_crystal_02, bone2Name=dyn_center, swing1LimitDegrees=30, swing2LimitDegrees=30, twistLimitDegrees=45}
 {bone1Name=dyn_crystal_03, bone2Name=dyn_center, swing1LimitDegrees=30, swing2LimitDegrees=30, twistLimitDegrees=45}
 {bone1Name=dyn_crystal_04, bone2Name=dyn_center, swing1LimitDegrees=30, swing2LimitDegrees=30, twistLimitDegrees=45}
]
