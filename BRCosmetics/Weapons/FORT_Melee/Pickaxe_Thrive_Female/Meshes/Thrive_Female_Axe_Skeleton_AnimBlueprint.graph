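# Animation Blueprint for the Thrive Female pickaxe that adds floating and physics-based motion to its five crystal bones.
# Two RigidBody nodes (world and base-bone space) feed additive bone offsets, a fixed one then a bobbing one, followed by per-crystal spring AnimDynamics.
# Each update builds three bobbing vectors from game-time sine waves.
# It also sets dynAlpha and rigidAlpha by frontend display context, though no visible node reads them.
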AnimBlueprint /BRCosmetics/Weapons/FORT_Melee/Pickaxe_Thrive_Female/Meshes/Thrive_Female_Axe_Skeleton_AnimBlueprint
target_skeleton=&/Thrive_Female_Axe_Skeleton
graphs [
 AnimGraph {nodes [
   n1 = RigidBody(ComponentPose=Local To Component(LocalPose=Input Pose())) {
    alpha=0
    bClampLinearTranslationLimitToRefPose=true
    bOverrideWorldGravity=true
    overridePhysicsAsset=&/Thrive_Female_Axe_PhysicsAsset
    simulationSpace=WorldSpace
    alphaBoolBlend {blendOption=Linear}
    baseBoneRef {boneName=Melee_TwoHanded_Handle}
    componentLinearAccScale=<0.1 0.1 0.1>
    componentLinearVelScale=<0.05 0.05 0.05>
    overrideWorldGravity {z=-980}
    simSpaceSettings {maxAngularAcceleration=5, maxAngularVelocity=5, maxLinearAcceleration=5, maxLinearVelocity=5, externalLinearDragV=<50 50 50>}
   }
   n2 = RigidBody(ComponentPose=@n1) {
    alpha=0
    bClampLinearTranslationLimitToRefPose=true
    bOverrideWorldGravity=true
    overridePhysicsAsset=&Thrive_Female_Axe_PhysicsAsset
    simulationSpace=BaseBoneSpace
    alphaBoolBlend {blendOption=Linear}
    baseBoneRef {boneName=Melee_TwoHanded_Handle}
    componentAppliedLinearAccClamp=<3000 3000 3000>
    componentLinearAccScale=<0.4 0.3 0.3>
    componentLinearVelScale=<0.4 0.3 0.3>
    overrideWorldGravity {z=-980}
   }
   n3 = "Transform (Modify) Bone - Bone: dyn_crystal_05"(ComponentPose=@n2, Alpha=1, Translation="3,0,0", Rotation="0,0,0", Scale="1,1,1") {
    translationMode=BMM_Additive
    translationSpace=BCS_BoneSpace
    alphaBoolBlend {blendOption=Linear}
    boneToModify {boneName=dyn_crystal_05}
    translation {x=3}
   }
   n4 = "Transform (Modify) Bone - Bone: dyn_crystal_04"(ComponentPose=@n3, Alpha=1, Translation="3,0,0", Rotation="0,0,0", Scale="1,1,1") {
    translationMode=BMM_Additive
    translationSpace=BCS_BoneSpace
    alphaBoolBlend {blendOption=Linear}
    boneToModify {boneName=dyn_crystal_04}
    translation {x=3}
   }
   n5 = "Transform (Modify) Bone - Bone: dyn_crystal_03"(ComponentPose=@n4, Alpha=1, Translation="3,0,0", Rotation="0,0,0", Scale="1,1,1") {
    translationMode=BMM_Additive
    translationSpace=BCS_BoneSpace
    alphaBoolBlend {blendOption=Linear}
    boneToModify {boneName=dyn_crystal_03}
    translation {x=3}
   }
   n6 = "Transform (Modify) Bone - Bone: dyn_crystal_02"(ComponentPose=@n5, Alpha=1, Translation="3,0,0", Rotation="0,0,0", Scale="1,1,1") {
    translationMode=BMM_Additive
    translationSpace=BCS_BoneSpace
    alphaBoolBlend {blendOption=Linear}
    boneToModify {boneName=dyn_crystal_02}
    translation {x=3}
   }
   n7 = "Transform (Modify) Bone - Bone: dyn_crystal_01"(ComponentPose=@n6, Alpha=1, Translation="3,0,0", Rotation="0,0,0", Scale="1,1,1") {
    translationMode=BMM_Additive
    translationSpace=BCS_BoneSpace
    alphaBoolBlend {blendOption=Linear}
    boneToModify {boneName=dyn_crystal_01}
    translation {x=3}
   }
   n8 = "Transform (Modify) Bone - Bone: dyn_crystal_05"(ComponentPose=@n7, Alpha=1, Translation=Get Floater3() -> (Floater3:Vector), Rotation="0,0,0", Scale="1,1,1") {translationMode=BMM_Additive, translationSpace=BCS_BoneSpace, alphaBoolBlend {blendOption=Linear}, boneToModify {boneName=dyn_crystal_05}}
   n9 = Get Floater2() -> (Floater2:Vector)
   n10 = "Transform (Modify) Bone - Bone: dyn_crystal_04"(ComponentPose=@n8, Alpha=1, Translation=@n9, Rotation="0,0,0", Scale="1,1,1") {translationMode=BMM_Additive, translationSpace=BCS_BoneSpace, alphaBoolBlend {blendOption=Linear}, boneToModify {boneName=dyn_crystal_04}}
   n11 = Get Floater1() -> (Floater1:Vector)
   n12 = "Transform (Modify) Bone - Bone: dyn_crystal_03"(ComponentPose=@n10, Alpha=1, Translation=@n11, Rotation="0,0,0", Scale="1,1,1") {translationMode=BMM_Additive, translationSpace=BCS_BoneSpace, alphaBoolBlend {blendOption=Linear}, boneToModify {boneName=dyn_crystal_03}}
   n13 = "Transform (Modify) Bone - Bone: dyn_crystal_02"(ComponentPose=@n12, Alpha=1, Translation=@n9, Rotation="0,0,0", Scale="1,1,1") {translationMode=BMM_Additive, translationSpace=BCS_BoneSpace, alphaBoolBlend {blendOption=Linear}, boneToModify {boneName=dyn_crystal_02}}
   n14 = "Transform (Modify) Bone - Bone: dyn_crystal_01"(ComponentPose=@n13, Alpha=1, Translation=@n11, Rotation="0,0,0", Scale="1,1,1") {translationMode=BMM_Additive, translationSpace=BCS_BoneSpace, alphaBoolBlend {blendOption=Linear}, boneToModify {boneName=dyn_crystal_01}}
   n15 = Anim Dynamics - Bone: dyn_crystal_01(ComponentPose=@n14, Alpha=1, ExternalForce="0,0,0") {
    alpha=0
    angularDampingOverride=0.9999
    bAngularSpring=true
    bGravityOverrideInSimSpace=true
    bLinearSpring=true
    bOverrideLinearDamping=true
    bUseGravityOverride=true
    linearDampingOverride=1
    linearSpringConstant=74
    alphaBoolBlend {blendOption=Linear}
    boundBone {boneName=dyn_crystal_01}
    componentAppliedLinearAccClamp=<3000 3000 3000>
    componentLinearAccScale=<0.4 0.4 0.4>
    componentLinearVelScale=<0.4 0.4 0.4>
    gravityOverride {z=-980}
    physicsBodyDefinitions [{
      sphereCollisionRadius=0
      boundBone {boneName=dyn_crystal_01}
      boxExtents=<15 15 15>
      constraintSetup {angularLimitsMax=<45 50 50>, angularLimitsMin=<-45 -50 -50>, angularTarget=<0 0 0>, linearAxesMax=<1 10 0>, linearAxesMin=<0 -10 0>}
      localJointOffset=<0 0 0>
     }]
   }
   n16 = Anim Dynamics - Bone: dyn_crystal_02(ComponentPose=@n15, Alpha=1, ExternalForce="0,0,0") {
    alpha=0
    angularDampingOverride=0.9999
    bAngularSpring=true
    bGravityOverrideInSimSpace=true
    bLinearSpring=true
    bOverrideLinearDamping=true
    bUseGravityOverride=true
    linearDampingOverride=1
    linearSpringConstant=98
    alphaBoolBlend {blendOption=Linear}
    boundBone {boneName=dyn_crystal_02}
    componentAppliedLinearAccClamp=<3000 3000 3000>
    componentLinearAccScale=<0.4 0.4 0.4>
    componentLinearVelScale=<0.4 0.4 0.4>
    gravityOverride {z=-980}
    physicsBodyDefinitions [{
      sphereCollisionRadius=0
      boundBone {boneName=dyn_crystal_02}
      boxExtents=<15 15 15>
      constraintSetup {angularLimitsMax=<45 50 50>, angularLimitsMin=<-45 -50 -50>, angularTarget=<0 0 0>, linearAxesMax=<1 10 0>, linearAxesMin=<0 -10 0>}
      localJointOffset=<0 0 0>
     }]
   }
   n17 = Anim Dynamics - Bone: dyn_crystal_03(ComponentPose=@n16, Alpha=1, ExternalForce="0,0,0") {
    alpha=0
    angularDampingOverride=0.9999
    bAngularSpring=true
    bGravityOverrideInSimSpace=true
    bLinearSpring=true
    bOverrideLinearDamping=true
    bUseGravityOverride=true
    linearDampingOverride=1
    linearSpringConstant=76
    alphaBoolBlend {blendOption=Linear}
    boundBone {boneName=dyn_crystal_03}
    componentAppliedLinearAccClamp=<3000 3000 3000>
    componentLinearAccScale=<0.4 0.4 0.4>
    componentLinearVelScale=<0.4 0.4 0.4>
    gravityOverride {z=-980}
    physicsBodyDefinitions [{
      sphereCollisionRadius=0
      boundBone {boneName=dyn_crystal_03}
      boxExtents=<15 15 15>
      constraintSetup {angularLimitsMax=<45 50 50>, angularLimitsMin=<-45 -50 -50>, angularTarget=<0 0 0>, linearAxesMax=<1 10 0>, linearAxesMin=<0 -10 0>}
      localJointOffset=<0 0 0>
     }]
   }
   n18 = Anim Dynamics - Bone: dyn_crystal_04(ComponentPose=@n17, Alpha=1, ExternalForce="0,0,0") {
    alpha=0
    angularDampingOverride=0.9999
    bAngularSpring=true
    bGravityOverrideInSimSpace=true
    bLinearSpring=true
    bOverrideLinearDamping=true
    bUseGravityOverride=true
    linearDampingOverride=1
    linearSpringConstant=75
    alphaBoolBlend {blendOption=Linear}
    boundBone {boneName=dyn_crystal_04}
    componentAppliedLinearAccClamp=<3000 3000 3000>
    componentLinearAccScale=<0.4 0.4 0.4>
    componentLinearVelScale=<0.4 0.4 0.4>
    gravityOverride {z=-980}
    physicsBodyDefinitions [{
      sphereCollisionRadius=0
      boundBone {boneName=dyn_crystal_04}
      boxExtents=<15 15 15>
      constraintSetup {angularLimitsMax=<45 50 50>, angularLimitsMin=<-45 -50 -50>, angularTarget=<0 0 0>, linearAxesMax=<1 10 0>, linearAxesMin=<0 -10 0>}
      localJointOffset=<0 0 0>
     }]
   }
   n19 = Anim Dynamics - Bone: dyn_crystal_05(ComponentPose=@n18, Alpha=1, ExternalForce="0,0,0") {
    alpha=0
    angularDampingOverride=1
    bAngularSpring=true
    bGravityOverrideInSimSpace=true
    bLinearSpring=true
    bOverrideLinearDamping=true
    bUseGravityOverride=true
    linearDampingOverride=1
    linearSpringConstant=75
    alphaBoolBlend {blendOption=Linear}
    boundBone {boneName=dyn_crystal_05}
    componentAppliedLinearAccClamp=<3000 3000 3000>
    componentLinearAccScale=<0.4 0.4 0.4>
    componentLinearVelScale=<0.4 0.4 0.4>
    gravityOverride {z=-980}
    physicsBodyDefinitions [{
      sphereCollisionRadius=0
      boundBone {boneName=dyn_crystal_05}
      boxExtents=<15 15 15>
      constraintSetup {angularLimitsMax=<45 20 20>, angularLimitsMin=<-45 -20 -20>, angularTarget=<0 0 0>, linearAxesMax=<5 0 10>, linearAxesMin=<0 0 -10>}
      localJointOffset=<0 0 0>
     }]
   }
   Output Pose(Result=Component To Local(ComponentPose=@n19)) {layerGroup=None}
  ]}
 EventGraph {nodes [
   Event BlueprintUpdateAnimation()
   n1 = GetGameTimeInSeconds()
   n2 = "Sin (Radians)"(A=float * float(A=float * float(A=0.8, B=2.8), B=@n1))
   > Set Floater1(Floater1:Vector=MakeVector(X=float * float(A=@n2, B=0.7), Y=0)) -> (Output_Get:Vector)
   n3 = "Sin (Radians)"(A=float * float(A=float * float(A=0.8, B=3.3), B=@n1))
   > Set Floater2(Floater2:Vector=MakeVector(X=float * float(A=@n3, B=0.7), Y=0)) -> (Output_Get:Vector)
   n4 = "Sin (Radians)"(A=float * float(A=float * float(A=1, B=3), B=@n1))
   > Set Floater3(Floater3:Vector=MakeVector(X=float * float(A=@n4, B=0.7), Y=0)) -> (Output_Get:Vector)
   > n5 = Branch(Condition=IsCosmeticObjectDisplayedInFrontendContext!(CosmeticObject:Object Reference) -> (ReturnValue:bool))
   > Set dynAlpha(dynAlpha:double=0.25) -> (Output_Get:double)
   > Set rigidAlpha(rigidAlpha:double=0) -> (Output_Get:double)
   Set dynAlpha(execute=@n5.else, dynAlpha=0)
   > Set rigidAlpha(rigidAlpha=1)
   TryGetPawnOwner()
   Sequence!(execute) -> (then_0, then_1, then_2:Exec)
  ]}
]
variables [
 customSpeed {type=number}
 dynAlpha {type=number}
 Floater1 {type=Vector}
 Floater2 {type=Vector}
 Floater3 {type=Vector}
 rigidAlpha {type=number}
]
